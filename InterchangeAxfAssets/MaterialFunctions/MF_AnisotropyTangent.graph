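# Material function for AxF materials that produces the anisotropy Tangent direction and Anisotropy strength.
# A Custom HLSL node converts the AnisotropicRotation input to an angle and rotates within the tangent/bitangent plane, using fixed axes for tangent-space normals and surface vectors otherwise.
# Anisotropy is the difference of AlphaX and AlphaY over their sum, or zero when the HasAnisotropy static switch is off.
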MaterialFunction /InterchangeAxfAssets/MaterialFunctions/MF_AnisotropyTangent
graph {nodes [
  MaterialExpressionFunctionInput_6 = FunctionInput() {inputName=AnisotropicRotation, inputType=FunctionInput_Scalar}
  n1 = Custom!(Rotation=@MaterialExpressionFunctionInput_6) {
   code="{\r\nconst float AngleRad = -(Rotation * 3.141592 - 1.570796);\r\n##if MATERIAL_TANGENTSPACENORMAL\r\n\tconst float3 T = float3(1,0,0);\r\n\tconst float3 B = float3(0,-1,0);\r\n\tconst float3 N = float3(0,0,1);\r\n##else\r\n\tconst float3 N = SurfaceNormal;\r\n\tconst float3 T = normalize(SurfaceTangent);\r\n\tconst float3 B = normalize(cross(SurfaceTangent, SurfaceNormal));\r\n##endif\r\n\tconst float3 Out = cos(AngleRad) * T + sin(AngleRad) * B;\r\n\treturn Out;\r\n}\r\n"
   showCode=true
   inputs [{inputName=Rotation, input {expression=&self:MaterialExpressionFunctionInput_6}}]
  }
  FunctionOutput(None=@n1) {outputName=Tangent, sortPriority=1}
  n2 = FunctionInput() {inputName=AlphaX, inputType=FunctionInput_Scalar, sortPriority=6}
  n3 = FunctionInput() {inputName=AlphaY, inputType=FunctionInput_Scalar, sortPriority=7}
  n4 = Subtract(A=@n2, B=@n3)
  n5 = Add(A=@n2, B=@n3)
  FunctionOutput(None=StaticSwitchParameter(True=Divide(A=@n4, B=@n5), False=Constant()) {group=Anisotropy, parameterName=HasAnisotropy, sortPriority=30}) {outputName=Anisotropy}
 ]}
