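# MF_DefaultBlend is a material function in InterchangeAxfAssets that turns an input Normal into blend weights, apparently for axis-projection (triplanar-style) blending.
# It raises each component to the fourth power by squaring twice, then divides by the component sum, taken as a dot product with one, so the weights total one.
# A zero vector is added, and a VertexInterpolator moves the calculation to the vertex shader for pixel-shader use.
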MaterialFunction /InterchangeAxfAssets/MaterialFunctions/MF_DefaultBlend
graph {nodes [
  n1 = FunctionInput() {inputName=Normal}
  n2 = Multiply(A=@n1, B=@n1)
  n3 = Multiply(A=@n2, B=@n2)
  n4 = Divide(A=@n3, B=DotProduct(A=@n3, B=Constant() {r=1}))
  n5 = Constant3Vector() {constant=<rgba 0 0 0 0>}
  FunctionOutput(None=VertexInterpolator(VS=Add(A=@n4, B=@n5)).PS)
 ]}
