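# Material function that prepares scaled, offset and animated coordinates plus a rotation value, named for tri-planar UV mapping in the AxF Interchange assets.
# Position is multiplied by 0.01, divided by overall scale times per-axis stretch, then shifted by offset and time-driven speed, with X negated for both.
# A second output turns degree-based rotation and time-driven spin into wrapped fractions of a full turn, leaving the actual rotation to the caller.
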MaterialFunction /InterchangeAxfAssets/MaterialFunctions/MF_TriPlanarUV_Coods
graph {nodes [
  n1 = Multiply(A=FunctionInput() {inputName=Position, sortPriority=12}) {constB=0.01, desc=Convert to 1 cm}
  n2 = Constant() {r=1}
  n3 = FunctionInput(Preview=@n2) {description=360 degree, inputName=Stretch on X, inputType=FunctionInput_Scalar, sortPriority=1}
  n4 = FunctionInput(Preview=@n2) {description=360 degree, inputName=Stretch on Y, inputType=FunctionInput_Scalar, sortPriority=2}
  n5 = Multiply(
   A=FunctionInput(Preview=@n2) {description=360 degree, inputName=Scale, inputType=FunctionInput_Scalar}
   B=AppendVector(
    A=AppendVector(A=@n3, B=@n4)
    B=FunctionInput() {description=360 degree, inputName=Stretch on Z, inputType=FunctionInput_Scalar, sortPriority=3}
   )
  )
  n6 = Constant()
  n7 = FunctionInput(Preview=@n6) {description=360 degree, inputName=Offset on X, inputType=FunctionInput_Scalar, sortPriority=5}
  n8 = AppendVector(
   A=AppendVector(
    A=Multiply(A=@n7) {constB=-1}
    B=FunctionInput(Preview=@n6) {description=360 degree, inputName=Offset on Y, inputType=FunctionInput_Scalar, sortPriority=6}
   )
   B=FunctionInput(Preview=@n6) {description=360 degree, inputName=Offset on Z, inputType=FunctionInput_Scalar, sortPriority=7}
  )
  n9 = Multiply(A=FunctionInput(Preview=@n6) {description=360 degree, inputName=Spped on X, inputType=FunctionInput_Scalar, sortPriority=8}) {constB=-1}
  n10 = FunctionInput(Preview=@n6) {description=360 degree, inputName=Speed on Y, inputType=FunctionInput_Scalar, sortPriority=9}
  n11 = Multiply(
   A=AppendVector(
    A=AppendVector(A=@n9, B=@n10)
    B=FunctionInput(Preview=@n6) {description=360 degree, inputName=Speed on Z, inputType=FunctionInput_Scalar, sortPriority=10}
   )
   B=NamedRerouteUsage()
  )
  FunctionOutput(None=Add(A=Divide(A=@n1, B=@n5), B=Add(A=@n8, B=@n11))) {outputName=WorldCoords_Scale_Offset}
  n12 = FunctionInput(Preview=@n6) {description=360 degree, inputName=Speed Rotation, inputType=FunctionInput_Scalar, sortPriority=11}
  n13 = Constant() {r=2}
  n14 = Multiply(
   A=Divide(
    A=MaterialFunctionCall!(Degrees=@n12) {materialFunction=/Engine/Functions/Engine_MaterialFunctions01/Math/DegreesToRadians}.Radians
    B=MaterialFunctionCall(Multiplier=@n13) {materialFunction=&Pi}.Result
   )
   B=NamedRerouteUsage()
  )
  n15 = NamedRerouteUsage()
  FunctionOutput(None=Frac(None=Add(A=@n14, B=@n15))) {outputName=WorldCoords_Rotation, sortPriority=2}
  n16 = FunctionInput(Preview=@n6) {description=360 degree, inputName=Rotation, inputType=FunctionInput_Scalar, sortPriority=4}
  n17 = Constant() {r=2}
  NamedRerouteDeclaration(None=Divide(
    A=MaterialFunctionCall(Degrees=@n16) {materialFunction=&DegreesToRadians}.Radians
    B=MaterialFunctionCall(Multiplier=@n17) {materialFunction=&Pi}.Result
   )) {name=UV_Rot, nodeColor=<rgba 1 0.2578 0 1>}
  NamedRerouteDeclaration(None=Time() {bOverride_Period=true, period=9999999}) {name=Time, nodeColor=<rgba 0.3828 1 0 1>}
 ]}
