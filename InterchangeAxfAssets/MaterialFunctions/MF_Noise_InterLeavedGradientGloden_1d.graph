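# Material function that produces a 1D interleaved gradient noise value per screen pixel, varied over time.
# A custom HLSL node hashes the pixel position with the constants from RandomInterleavedGradientNoise.ush.
# It then adds the frame index times the golden ratio and takes the fraction, giving a low-discrepancy temporal sequence.
# Inputs default to screen pixel position and View.StateFrameIndex modulo 64.
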MaterialFunction /InterchangeAxfAssets/MaterialFunctions/MF_Noise_InterLeavedGradientGloden_1d
graph {nodes [
  MaterialExpressionFunctionInput_0 = FunctionInput(Preview=Custom() {code="View.StateFrameIndex%64;", description=ViewStateFrameIndexMod64, outputType=CMOT_Float1, inputs []}) {bUsePreviewValueAsDefault=true, inputName=TemporalSampleIndex, inputType=FunctionInput_Scalar}
  MaterialExpressionFunctionInput_1 = FunctionInput(Preview=ScreenPosition().PixelPosition) {bUsePreviewValueAsDefault=true, inputName=PixelPosition, inputType=FunctionInput_Vector2, sortPriority=1}
  n1 = Custom!(frame=@MaterialExpressionFunctionInput_0, pixel=@MaterialExpressionFunctionInput_1) {
   code="//From RandomInterleavedGradientNoise.ush\r\nconst float3 magic = float3( 0.06711056f, 0.00583715f, 52.9829189f ); \r\nfloat ign = frac(magic.z * frac(dot(pixel, magic.xy)));\r\n//Add golden ratio for low discrepancy\r\nfloat gr = frame * (1.0f + sqrt(5.0f)) / 2.0f;\r\nreturn frac(ign + gr);"
   description=InterleavedGradientNoise
   outputType=CMOT_Float1
   showCode=true
   inputs [
    {
     inputName=frame
     input {expression=&self:MaterialExpressionFunctionInput_0}
    }
    {
     inputName=pixel
     input {expression=&self:MaterialExpressionFunctionInput_1}
    }
   ]
  }
  FunctionOutput(None=@n1)
 ]}
