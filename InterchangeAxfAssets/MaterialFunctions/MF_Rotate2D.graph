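# Material function in the InterchangeAxfAssets content that rotates 2D coordinates around a pivot, by default texture coordinates around the UV center.
# It subtracts Pivot from Coords, applies a sine and cosine rotation to the two components, then adds the pivot back.
# RotationAmount is saturated to 0-1 and, with the default trig period, spans one full turn, so values outside that range clamp rather than wrap.
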MaterialFunction /InterchangeAxfAssets/MaterialFunctions/MF_Rotate2D
graph {nodes [
  n1 = FunctionInput(Preview=Constant2Vector() {g=0.5, r=0.5}) {bUsePreviewValueAsDefault=true, inputName=Pivot, inputType=FunctionInput_Vector2, sortPriority=20}
  n2 = Saturate(None=FunctionInput() {bUsePreviewValueAsDefault=true, inputName=RotationAmount, inputType=FunctionInput_Scalar})
  n3 = Cosine(None=@n2)
  n4 = TextureCoordinate()
  n5 = MaterialFunctionCall(Float2=Subtract(A=FunctionInput(Preview=@n4) {bUsePreviewValueAsDefault=true, inputName=Coords, inputType=FunctionInput_Vector2, sortPriority=10}, B=@n1)) {materialFunction=&BreakOutFloat2Components}
  n6 = Sine(None=@n2)
  n7 = Subtract(A=Multiply(A=@n3, B=@n5.R), B=Multiply(A=@n6, B=@n5.G))
  n8 = Add(A=Multiply(A=@n3, B=@n5.G), B=Multiply(A=@n6, B=@n5.R))
  FunctionOutput(None=Add(A=@n1, B=AppendVector(A=@n7, B=@n8)))
 ]}
