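# Core triplanar projection material function for AxF materials that picks one of three planar UV sets per pixel instead of blending them.
# Noise-dithered Blend weights choose the axis, unless ProjectAxis forces one.
# Custom HLSL selects that axis's sign-corrected UV and screen-space derivatives and builds a normal-orthogonalized tangent and bitangent.
# It outputs the rotated UV, tangent frame, packed derivatives, axis index, planar UVs and dither pattern.
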MaterialFunction /InterchangeAxfAssets/MaterialFunctions/MF_Triplanar_Core
graph {nodes [
  FunctionOutput(None=NamedRerouteUsage()) {outputName=DitherPattern, sortPriority=20}
  MaterialExpressionNamedRerouteUsage_3 = NamedRerouteUsage()
  MaterialExpressionNamedRerouteUsage_7 = NamedRerouteUsage()
  MaterialExpressionNamedRerouteUsage_8 = NamedRerouteUsage()
  MaterialExpressionNamedRerouteUsage_9 = NamedRerouteUsage()
  MaterialExpressionNamedRerouteUsage_10 = NamedRerouteUsage()
  MaterialExpressionNamedRerouteUsage_11 = NamedRerouteUsage()
  MaterialExpressionFunctionInput_4 = FunctionInput() {inputName=Blend}
  n1 = MaterialFunctionCall!(TemporalSampleIndex, PixelPosition) {materialFunction=&/MF_Noise_InterLeavedGradientGloden_1d}
  MaterialExpressionMultiply_12 = Multiply(
   A=FunctionInput() {bUsePreviewValueAsDefault=true, inputName=TriplanarDitherPattern, inputType=FunctionInput_Scalar, sortPriority=5}
   B=Subtract(A=@n1.Result) {constB=0.5}
  )
  n2 = Custom!(blend=@MaterialExpressionFunctionInput_4, dither=@MaterialExpressionMultiply_12) {
   code="//modified from https://ryandowlingsoka.com/unreal/triplanar-dither-biplanar/\r\nblend = saturate(blend);\r\nconst float ditherx = blend.x - dither;\r\nconst float ditherz = blend.z - dither; \r\n\r\nint index = 0;\r\nindex = max(ditherx, ditherz) > blend.y ? 0 : 1;\r\nindex = ditherz > max(blend.x, blend.y) ? 2 : index;\r\n\r\nreturn index;"
   description=createTriplanarSelection
   outputType=CMOT_Float1
   showCode=true
   inputs [
    {inputName=blend, input {expression=&self:MaterialExpressionFunctionInput_4}}
    {inputName=dither, input {expression=&self:MaterialExpressionMultiply_12}}
   ]
  }
  n3 = FunctionInput() {bUsePreviewValueAsDefault=true, inputName=ProjectAxis, inputType=FunctionInput_Scalar, sortPriority=-5, previewValue=<-1 0 0 1>}
  n4 = Clamp(None=Floor(None=@n3)) {maxDefault=2, minDefault=-1}
  n5 = Multiply(A=Sign(None=@n4)) {constB=-1}
  MaterialExpressionFunctionOutput_5 = FunctionOutput(None=Max(A=Multiply(A=@n2, B=@n5), B=@n4)) {outputName=AxisIndex, sortPriority=90}
  MaterialExpressionNamedRerouteUsage_12 = NamedRerouteUsage()
  n6 = Custom!(
   uvX=@MaterialExpressionNamedRerouteUsage_3
   uvY=@MaterialExpressionNamedRerouteUsage_7
   uvZ=@MaterialExpressionNamedRerouteUsage_8
   dPdx=@MaterialExpressionNamedRerouteUsage_9
   dPdy=@MaterialExpressionNamedRerouteUsage_10
   signs=@MaterialExpressionNamedRerouteUsage_11
   selection=@MaterialExpressionFunctionOutput_5
   NormalV=@MaterialExpressionNamedRerouteUsage_12
  ) {
   code="//modified from https://ryandowlingsoka.com/unreal/triplanar-dither-biplanar/\r\n//further reading https://therobotseven.com/devlog/triplanar-texturing-derivative-maps/\r\nint index = trunc(selection + 0.01f);\r\n\r\nfloat2 uv_choices[3] = {uvX, uvY, uvZ};\r\nfloat2 deriv_choiceX[3] = {-dPdx.yz, float2(dPdx.x, -dPdx.z), dPdx.xy};\r\nfloat2 deriv_choiceY[3] = {-dPdy.yz, float2(dPdy.x, -dPdy.z), dPdy.xy};\r\n\r\nfloat2 flip = float2(1.f,signs[index]);\r\nfloat3 tanDir[3] = { float3(0.f, -1.f, 0.f), \r\n                      float3(1.f, 0.f, 0.f), \r\n                      float3(1.f, 0.f, 0.f) \r\n                   };\r\n\r\nfloat3 Tangent = tanDir[index] * signs[index];\r\n\r\n//Parameters.TangentToWorld[2] == vertexNormal\r\n\r\nTangent += NormalV * ( -1.f * dot( Tangent, NormalV) );\r\n\r\nout_tangent = normalize( Tangent );\r\nout_bitangent = normalize( cross( NormalV, Tangent ) ); \r\n\r\nout_packedDXY = float4( deriv_choiceX[index], deriv_choiceY[index] );\r\n\r\n//return is triplanar UVs\r\nreturn uv_choices[index] * signs[index] * flip;\r\n\r\n"
   description=SelectUVs
   outputType=CMOT_Float2
   showCode=true
   additionalOutputs [
    {outputName=out_tangent, outputType=CMOT_Float3}
    {outputName=out_bitangent, outputType=CMOT_Float3}
    {outputName=out_packedDXY, outputType=CMOT_Float4}
   ]
   inputs [
    {inputName=uvX, input {expression=&self:MaterialExpressionNamedRerouteUsage_3}}
    {inputName=uvY, input {expression=&self:MaterialExpressionNamedRerouteUsage_7}}
    {inputName=uvZ, input {expression=&self:MaterialExpressionNamedRerouteUsage_8}}
    {inputName=dPdx, input {expression=&self:MaterialExpressionNamedRerouteUsage_9}}
    {inputName=dPdy, input {expression=&self:MaterialExpressionNamedRerouteUsage_10}}
    {inputName=signs, input {expression=&self:MaterialExpressionNamedRerouteUsage_11}}
    {inputName=selection, input {expression=&self:MaterialExpressionFunctionOutput_5}}
    {inputName=NormalV, input {expression=&self:MaterialExpressionNamedRerouteUsage_12}}
   ]
  }
  FunctionOutput(None=MaterialFunctionCall!(
    RotationAmount=FunctionInput() {bUsePreviewValueAsDefault=true, inputName="Rotation (Turns)", inputType=FunctionInput_Scalar, sortPriority=-10}
    Coords=@n6.return
    Pivot=Constant2Vector() {g=0.5, r=0.5}
   ) {materialFunction=&/MF_Rotate2D}.Result) {outputName=triplanar_UV}
  FunctionOutput(None=@n6.out_tangent) {outputName=triplanar_tangentWS, sortPriority=60}
  FunctionOutput(None=@n6.out_bitangent) {outputName=triplanar_bitangentWS, sortPriority=70}
  FunctionOutput(None=@n6.out_packedDXY) {outputName=triplanar_packedDXY, sortPriority=10}
  NamedRerouteDeclaration(None=@n1.Result) {name=DitherPattern, nodeColor=<rgba 0.9688 1 0 1>}
  n7 = FunctionInput() {description=360 degree, inputName=AXF_X, inputType=FunctionInput_Scalar, sortPriority=9}
  n8 = FunctionInput() {description=360 degree, inputName=AXF_Y, inputType=FunctionInput_Scalar, sortPriority=10}
  NamedRerouteDeclaration(None=Multiply(A=AppendVector(A=@n7, B=@n8)) {constB=100}) {name=AXF_Parameters, nodeColor=<rgba 1 0.6797 0 1>}
  n9 = WorldPosition()
  n10 = Constant() {r=128}
  n11 = VertexInterpolator(VS=FunctionInput(Preview=Divide(A=@n9.XYZ, B=@n10)) {inputName=Coordinates, sortPriority=-30})
  n12 = MaterialFunctionCall(Float3=@n11.PS) {materialFunction=&BreakOutFloat3Components}
  n13 = Multiply(A=AppendVector(A=@n12.G, B=@n12.G)) {constB=-1}
  n14 = NamedRerouteUsage()
  NamedRerouteDeclaration(None=FunctionOutput(None=Multiply(A=@n13, B=@n14)) {outputName=Planar_uvX, sortPriority=30}) {name=uvX, nodeColor=<rgba 1 0 0.5391 1>}
  n15 = AppendVector(A=@n12.R, B=Multiply(A=@n12.B) {constB=-1})
  NamedRerouteDeclaration(None=FunctionOutput(None=Multiply(A=@n14, B=@n15)) {outputName=Planar_uvY, sortPriority=40}) {name=uvY, nodeColor=<rgba 1 0 0.5391 1>}
  n16 = AppendVector(A=@n12.R, B=@n12.R)
  NamedRerouteDeclaration(None=FunctionOutput(None=Multiply(A=@n14, B=@n16)) {outputName=Planar_uvZ, sortPriority=50}) {name=uvZ, nodeColor=<rgba 1 0 0.5391 1>}
  NamedRerouteDeclaration(None=DDX(Value=@n11.PS)) {name=DPDX, nodeColor=<rgba 0.5 0 1 1>}
  NamedRerouteDeclaration(None=DDY(Value=@n11.PS)) {name=DPDY, nodeColor=<rgba 0.5 0 1 1>}
  n17 = FunctionInput() {inputName=Normal}
  NamedRerouteDeclaration(None=Sign(None=NamedRerouteDeclaration(None=@n17) {name=InputNormal, nodeColor=<rgba 0 0.4609 1 1>})) {name=Sign, nodeColor=<rgba 0.5 0 1 1>}
 ]}
