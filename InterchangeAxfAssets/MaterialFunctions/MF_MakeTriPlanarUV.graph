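# Material function for Interchange AxF materials that builds triplanar projection UVs, with per-axis planar UVs, derivatives, tangent, bitangent, axis index, rotation and dither pattern.
# MF_TriPlanarUV_Coods turns position, scale, offset, rotation and animation speed inputs into transformed coordinates and rotation.
# MF_Triplanar_Core projects using these, the normal, AxF X/Y scales and a manual axis or three-axis blend with dither-controlled softness.
# Named reroutes carry results to outputs, with packed derivatives split into DDX and DDY.
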MaterialFunction /InterchangeAxfAssets/MaterialFunctions/MF_MakeTriPlanarUV
graph {nodes [
  FunctionOutput(None=NamedRerouteUsage()) {outputName=Planar_uvX, sortPriority=5}
  FunctionOutput(None=NamedRerouteUsage()) {outputName=Planar_uvY, sortPriority=6}
  FunctionOutput(None=NamedRerouteUsage()) {outputName=TriplanarUV_Tangent, sortPriority=10}
  FunctionOutput(None=NamedRerouteUsage()) {outputName=TriplanarUV_Bitangent, sortPriority=11}
  FunctionOutput(None=NamedRerouteUsage()) {outputName=DitherPattern, sortPriority=9}
  FunctionOutput(None=NamedRerouteUsage()) {outputName=Planar_uvZ, sortPriority=7}
  FunctionOutput(None=NamedRerouteUsage()) {outputName=TP_Axis, sortPriority=8}
  FunctionOutput(None=NamedRerouteUsage()) {outputName=UV, sortPriority=1}
  FunctionOutput(None=NamedRerouteUsage()) {outputName=Rotation, sortPriority=2}
  n1 = NamedRerouteUsage()
  FunctionOutput(None=ComponentMask(None=@n1) {g=true, r=true}) {outputName=DDX, sortPriority=3}
  FunctionOutput(None=ComponentMask(None=@n1) {a=true, b=true}) {outputName=DDY, sortPriority=4}
  NamedRerouteDeclaration(None=FunctionInput() {inputName=AXF_Scale_X, inputType=FunctionInput_Scalar, sortPriority=30}) {name=AXF_X, nodeColor=<rgba 0 1 0.6484 1>}
  n2 = FunctionInput() {
   desc="Specify the axis along which the textures are projected for the triplanar projection.  -1 will sample on all three axes and blend, 0 will project along the x-axis, 1 will project along the y-axis, and 2 will project along the z-axis."
   inputName=TriPlanarUV_ManualAxis
   inputType=FunctionInput_Scalar
   sortPriority=410
   previewValue=<-1 -1 -1 -1>
  }
  n3 = FunctionInput() {
   desc="Controls the smoothness of the blend between the axes of the triplanar projection where 0 = abrupt and 1 = smooth."
   inputName=TriPlanarUV_BlendSoftness
   inputType=FunctionInput_Scalar
   sortPriority=450
   previewValue=<0.1 0.1 0.1 0.1>
  }
  n4 = MaterialFunctionCall!(
   Coordinates=NamedRerouteUsage()
   "Rotation (Turns)"=NamedRerouteUsage()
   ProjectAxis=@n2
   Normal=FunctionInput() {inputName=TriPlanarUV_Normal, sortPriority=510}
   Blend=FunctionInput() {inputName=Blend, sortPriority=600}
   TriplanarDitherPattern=@n3
   AXF_X=NamedRerouteUsage()
   AXF_Y=NamedRerouteUsage()
  ) {materialFunction=&/MF_Triplanar_Core}
  NamedRerouteDeclaration(None=@n4.triplanar_tangentWS) {name=TP_Tangent, nodeColor=<rgba 1 0 0.8672 1>}
  NamedRerouteDeclaration(None=@n4.Planar_uvZ) {name=Planar_uvZ, nodeColor=<rgba 0.5 0 1 1>}
  NamedRerouteDeclaration(None=@n4.DitherPattern) {name=DitherPattern, nodeColor=<rgba 0.1484 1 0 1>}
  NamedRerouteDeclaration(None=@n4.triplanar_UV) {name=TP_UV, nodeColor=<rgba 1 0 0.8906 1>}
  n5 = FunctionInput() {inputName=UVDetail_Scale_X, inputType=FunctionInput_Scalar, sortPriority=200, previewValue=<1 1 1 1>}
  n6 = FunctionInput() {inputName=UVDetail_Scale_Y, inputType=FunctionInput_Scalar, sortPriority=210, previewValue=<1 1 1 1>}
  n7 = FunctionInput() {inputName=UVDetail_Scale_Z, inputType=FunctionInput_Scalar, sortPriority=220, previewValue=<1 1 1 1>}
  n8 = FunctionInput() {inputName=UVDetail_Offset_X, inputType=FunctionInput_Scalar, sortPriority=230, previewValue=<1 1 1 1>}
  n9 = FunctionInput() {inputName=UVDetail_Offset_Y, inputType=FunctionInput_Scalar, sortPriority=240, previewValue=<1 1 1 1>}
  n10 = FunctionInput() {inputName=UVDetail_Offset_Z, inputType=FunctionInput_Scalar, sortPriority=250, previewValue=<1 1 1 1>}
  n11 = FunctionInput() {inputName=UVAnimation_Speed_X, inputType=FunctionInput_Scalar, sortPriority=300, previewValue=<1 1 1 1>}
  n12 = FunctionInput() {inputName=UVAnimation_Speed_Y, inputType=FunctionInput_Scalar, sortPriority=310, previewValue=<1 1 1 1>}
  n13 = FunctionInput() {inputName=UVAnimation_Speed_Z, inputType=FunctionInput_Scalar, sortPriority=320, previewValue=<1 1 1 1>}
  n14 = FunctionInput() {inputName=UVAnimation_SpeedRotationDegree, inputType=FunctionInput_Scalar, sortPriority=330, previewValue=<1 1 1 1>}
  n15 = MaterialFunctionCall!(
   Scale=FunctionInput() {inputName=UV_Scale, inputType=FunctionInput_Scalar, previewValue=<1 1 1 1>}
   Stretch on X=@n5
   Stretch on Y=@n6
   Stretch on Z=@n7
   Rotation=FunctionInput() {inputName=UV_RotationDegree, inputType=FunctionInput_Scalar, sortPriority=20}
   Offset on X=@n8
   Offset on Y=@n9
   Offset on Z=@n10
   Spped on X=@n11
   Speed on Y=@n12
   Speed on Z=@n13
   Speed Rotation=@n14
   Position=FunctionInput() {inputName=TriPlanarUV_Position, sortPriority=500, previewValue=<1 1 1 1>}
  ) {materialFunction=&/MF_TriPlanarUV_Coods}
  NamedRerouteDeclaration(None=@n15.WorldCoords_Rotation) {name=TP_Rotation, nodeColor=<rgba 0.9453 1 0 1>}
  NamedRerouteDeclaration(None=@n4.triplanar_packedDXY) {name=TP_DXY, nodeColor=<rgba 0 1 0.7891 1>}
  NamedRerouteDeclaration(None=@n15.WorldCoords_Scale_Offset) {name=TP_Coords, nodeColor=<rgba 0.7344 0 1 1>}
  NamedRerouteDeclaration(None=@n4.triplanar_bitangentWS) {name=TP_Bitangent, nodeColor=<rgba 0 1 0.8125 1>}
  NamedRerouteDeclaration(None=FunctionInput() {inputName=AXF_Scale_Y, inputType=FunctionInput_Scalar, sortPriority=40}) {name=AXF_Y, nodeColor=<rgba 1 0.3281 0 1>}
  NamedRerouteDeclaration(None=@n4.Planar_uvX) {name=Planar_uvX, nodeColor=<rgba 0.5 0 1 1>}
  NamedRerouteDeclaration(None=@n4.AxisIndex) {name=TP_Axis, nodeColor=<rgba 0.1719 0 1 1>}
  NamedRerouteDeclaration(None=@n4.Planar_uvY) {name=Planar_uvY, nodeColor=<rgba 0.5 0 1 1>}
 ]}
