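# MF_smin is a material function in InterchangeAxfAssets that returns a smooth minimum of two scalars, A and B.
# A single Custom HLSL node applies Inigo Quilez's quadratic smin, credited in its code, subtracting a quadratic falloff from the hard minimum where the inputs are close.
# Tolerance sets the blend width, scaled inside the code and used as a divisor without a zero guard.
# All three inputs fall back to preview values when unconnected.
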MaterialFunction /InterchangeAxfAssets/MaterialFunctions/MF_smin
graph {nodes [
  MaterialExpressionFunctionInput_0 = FunctionInput() {bUsePreviewValueAsDefault=true, inputName=A, inputType=FunctionInput_Scalar}
  MaterialExpressionFunctionInput_1 = FunctionInput() {bUsePreviewValueAsDefault=true, inputName=B, inputType=FunctionInput_Scalar, sortPriority=1}
  MaterialExpressionFunctionInput_2 = FunctionInput() {bUsePreviewValueAsDefault=true, inputName=Tolerance, inputType=FunctionInput_Scalar, sortPriority=2, previewValue=<0.2 0 0 1>}
  n1 = Custom!(a=@MaterialExpressionFunctionInput_0, b=@MaterialExpressionFunctionInput_1, k=@MaterialExpressionFunctionInput_2) {
   code="//Credit here: https://iquilezles.org/articles/smin/\r\nk *= 4.f;\r\nfloat h = (max( k - abs(a-b), 0.f ) / k) - 1.f;\r\nreturn min(a,b) - k * ( h * ( 2.f + h ) + 1.f ) / 4.f; "
   description=quadratic_smin
   outputType=CMOT_Float1
   showCode=true
   inputs [
    {inputName=a, input {expression=&self:MaterialExpressionFunctionInput_0}}
    {inputName=b, input {expression=&self:MaterialExpressionFunctionInput_1}}
    {inputName=k, input {expression=&self:MaterialExpressionFunctionInput_2}}
   ]
  }
  FunctionOutput(None=@n1)
 ]}
