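# Material function that computes a normalized triplanar AxisMask from a normal input, modulated by a scalar height.
# The normal is raised to the fourth power per component, normalized by its sum, then rescaled and offset by BlendFactor.
# Each axis weight is combined with the shared offset Height through negated MF_smin calls, acting as a smooth maximum softened by BlendTolerance.
# The three results are renormalized to sum to one.
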MaterialFunction /InterchangeAxfAssets/MaterialFunctions/MF_TriplanarBlendHeightA
graph {nodes [
  n1 = FunctionInput() {bUsePreviewValueAsDefault=true, inputName=Height, inputType=FunctionInput_Scalar, previewValue=<0.5 0 0 1>}
  n2 = Subtract(A=@n1)
  n3 = Multiply(A=@n2) {constB=-1}
  n4 = FunctionInput() {inputName=NormalTP}
  n5 = Multiply(A=@n4, B=@n4)
  n6 = Multiply(A=@n5, B=@n5)
  n7 = Divide(A=@n6, B=DotProduct(A=@n6, B=Constant3Vector() {constant=<rgba 1 1 1 0>})) {desc=blend}
  n8 = MaterialFunctionCall(Float3=Subtract(A=Multiply(A=@n7) {constB=2}) {constB=0.667}) {materialFunction=&BreakOutFloat3Components}
  n9 = FunctionInput() {bUsePreviewValueAsDefault=true, inputName=BlendFactor, inputType=FunctionInput_Scalar, sortPriority=10}
  n10 = Multiply(A=Add(A=@n8.R, B=@n9)) {constB=-1}
  n11 = FunctionInput() {bUsePreviewValueAsDefault=true, inputName=BlendTolerance, inputType=FunctionInput_Scalar, sortPriority=20, previewValue=<0.25 0 0 1>}
  n12 = Multiply(A=@n2) {constB=-1}
  n13 = Multiply(A=Add(A=@n8.G, B=@n9)) {constB=-1}
  n14 = Multiply(A=@n2) {constB=-1}
  n15 = Multiply(A=Add(A=@n8.B, B=@n9)) {constB=-1}
  n16 = MaterialFunctionCall(
   X=Multiply(A=MaterialFunctionCall!(A=@n3, B=@n10, Tolerance=@n11) {materialFunction=&/MF_smin}.Result) {constB=-1}
   Y=Multiply(A=MaterialFunctionCall(A=@n12, B=@n13, Tolerance=@n11) {materialFunction=&MF_smin}.Result) {constB=-1}
   Z=Multiply(A=MaterialFunctionCall(A=@n14, B=@n15, Tolerance=@n11) {materialFunction=&MF_smin}.Result) {constB=-1}
  ) {materialFunction=&MakeFloat3}
  n17 = Constant3Vector() {constant=<rgba 1 1 1 0>}
  FunctionOutput(None=Divide(A=@n16.Result, B=DotProduct(A=@n16.Result, B=@n17)) {desc=blend}) {outputName=AxisMask}
 ]}
