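# MF_POMOverrideBlend is an InterchangeAxfAssets material function that computes a height-based triplanar axis mask from a parallax occlusion (POM) texture.
# It samples the texture at the input UV with explicit DDX/DDY derivatives and isolates the red channel via a dot product.
# That height, scaled by BlendStrength, feeds MF_TriplanarBlendHeightA along with Normal, BlendFactor and blend tolerance inputs, and the resulting AxisMask is the sole output.
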MaterialFunction /InterchangeAxfAssets/MaterialFunctions/MF_POMOverrideBlend
graph {nodes [
  n1 = TextureObject()
  n2 = DotProduct(
   A=TextureSample!(
    UVs=FunctionInput() {inputName=UV, inputType=FunctionInput_Vector2, sortPriority=1}
    Tex=FunctionInput(Preview=@n1) {inputName=POMTex, inputType=FunctionInput_Texture2D}
    "DDX(UVs)"=FunctionInput() {inputName=DDX, inputType=FunctionInput_Vector2, sortPriority=2}
    "DDY(UVs)"=FunctionInput() {inputName=DDY, inputType=FunctionInput_Vector2, sortPriority=3}
    Apply View MipBias
   ) {mipValueMode=TMVM_Derivative}.RGBA
   B=Constant4Vector() {constant=<rgba 1 0 0 0>}.RGBA
  )
  n3 = FunctionInput() {inputName=BlendStrength, inputType=FunctionInput_Scalar, sortPriority=5}
  FunctionOutput(None=MaterialFunctionCall!(
    NormalTP=FunctionInput() {inputName=Normal, sortPriority=4}
    Height=Multiply(A=@n2, B=@n3)
    BlendFactor=FunctionInput() {inputName=BlendFactor, inputType=FunctionInput_Scalar, sortPriority=6}
    BlendTolerance=FunctionInput() {inputName=BlendTorlerance, inputType=FunctionInput_Scalar, sortPriority=7}
   ) {materialFunction=&/MF_TriplanarBlendHeightA}.AxisMask)
 ]}
