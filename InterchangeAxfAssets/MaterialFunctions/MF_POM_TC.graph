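# Material function for the Interchange AxF assets that turns input UVs into parallax occlusion mapped UVs from a height texture.
# Custom HLSL ray-marches the red channel using explicit UV gradients and interpolates the hit between steps.
# RayUV and rescaled height strength drive the offset, RefPlanar shifts the start, and FresnelVX blends step count from MaxStep toward one.
# A ray tracing quality switch skips the march, keeping only the shifted UVs.
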MaterialFunction /InterchangeAxfAssets/MaterialFunctions/MF_POM_TC
graph {nodes [
  n1 = Multiply(A=FunctionInput() {inputName=POM_HeightStrength, inputType=FunctionInput_Scalar, sortPriority=12}) {constB=0.01, desc=rescale}
  n2 = NamedRerouteUsage()
  n3 = Multiply(A=OneMinus(None=FunctionInput() {inputName=RefPlanar, inputType=FunctionInput_Scalar, sortPriority=14})) {constB=-1}
  n4 = FunctionInput(Preview=TextureCoordinate()) {inputName=UV, inputType=FunctionInput_Vector2, sortPriority=11}
  n5 = Add(A=Multiply(A=@n1, B=Multiply(A=@n2, B=@n3)), B=@n4)
  MaterialExpressionFunctionInput_0 = FunctionInput(Preview=TextureObject() {samplerType=SAMPLERTYPE_LinearColor, texture=/InterchangeAssets/gltf/Textures/T_Gray_Linear}) {inputName=POM_HeightTexture, inputType=FunctionInput_Texture2D}
  MaterialExpressionReroute_4 = Reroute(None=@n5)
  MaterialExpressionNamedRerouteUsage_5 = NamedRerouteUsage()
  MaterialExpressionNamedRerouteUsage_6 = NamedRerouteUsage()
  MaterialExpressionNamedRerouteUsage_4 = NamedRerouteUsage()
  MaterialExpressionDDX_0 = DDX(Value=@n4)
  MaterialExpressionDDY_0 = DDY(Value=@n4)
  MaterialExpressionConstant4Vector_1 = Constant4Vector() {constant=<rgba 1 0 0 0>}
  n6 = Custom!(
   Tex=@MaterialExpressionFunctionInput_0
   UV=@MaterialExpressionReroute_4
   MaxSteps=@MaterialExpressionNamedRerouteUsage_5
   stepsize=@MaterialExpressionNamedRerouteUsage_6
   UVDist=@MaterialExpressionNamedRerouteUsage_4
   InDDX=@MaterialExpressionDDX_0
   InDDY=@MaterialExpressionDDY_0
   HeightMapChannel=@MaterialExpressionConstant4Vector_1.RGBA
  ) {
   code="float rayheight = 1;\r\nfloat oldray = 1;\r\nfloat2 offset = 0;\r\nfloat oldtex = 1;\r\nfloat texAtRay;\r\nfloat yintersect;\r\nint i = 0;\r\n\r\nwhile (i < MaxSteps + 2)\r\n{\r\n\tfloat texAtRay = dot(Tex.SampleGrad( GetMaterialSharedSampler(TexSampler,View.MaterialTextureBilinearWrapedSampler), UV + offset, \r\n                     InDDX * (MaterialFloat2)View.MaterialTextureDerivativeMultiply, InDDY * (MaterialFloat2)View.MaterialTextureDerivativeMultiply ), HeightMapChannel);\r\n\r\n\tif (rayheight < texAtRay)\r\n\t{\r\n\t\tfloat xintersect = (oldray - oldtex) + (texAtRay - rayheight);\r\n\t\txintersect = (texAtRay - rayheight) / xintersect;\r\n\t\tyintersect = ( oldray * (xintersect) ) + ( rayheight * (1 - xintersect) );\r\n\t\toffset -= (xintersect * UVDist);\r\n\t\tbreak;\r\n\t}\r\n\r\n\toldray = rayheight;\r\n\trayheight -= stepsize;\r\n\toffset += UVDist;\r\n\toldtex = texAtRay;\r\n\r\n\ti++;\r\n}\r\n\r\nreturn offset;"
   description=Parallax Only Mapping
   outputType=CMOT_Float2
   showCode=true
   inputs [
    {inputName=Tex, input {expression=&self:MaterialExpressionFunctionInput_0}}
    {inputName=UV, input {expression=&self:MaterialExpressionReroute_4}}
    {inputName=MaxSteps, input {expression=&self:MaterialExpressionNamedRerouteUsage_5}}
    {inputName=stepsize, input {expression=&self:MaterialExpressionNamedRerouteUsage_6}}
    {inputName=UVDist, input {expression=&self:MaterialExpressionNamedRerouteUsage_4}}
    {inputName=InDDX, input {expression=&self:MaterialExpressionDDX_0}}
    {inputName=InDDY, input {expression=&self:MaterialExpressionDDY_0}}
    {
     inputName=HeightMapChannel
     input {expression=&self:MaterialExpressionConstant4Vector_1, mask=1, maskA=1, maskB=1, maskG=1, maskR=1}
    }
   ]
  }
  FunctionOutput(None=RayTracingQualitySwitch(Normal=Add(A=@n5, B=@n6), RayTraced=@n5)) {outputName=Parallax UVs}
  n7 = LinearInterpolate(
   A=FunctionInput(Preview=Constant() {r=32}) {inputName=MaxStep, inputType=FunctionInput_Scalar, sortPriority=13}
   B=Constant() {desc=MinStep, r=1}
   Alpha=FunctionInput() {inputName=FresnelVX, inputType=FunctionInput_Scalar, sortPriority=20}
  )
  NamedRerouteDeclaration(None=Divide(B=@n7)) {name=StepSize, nodeColor=<rgba 1 0.7266 0 1>}
  n8 = NamedRerouteUsage()
  NamedRerouteDeclaration(None=Multiply(A=Multiply(A=@n8, B=@n1), B=NamedRerouteUsage())) {name=UV_D, nodeColor=<rgba 0.4062 0 1 1>}
  NamedRerouteDeclaration(None=Floor(None=@n7)) {name=Maxstep, nodeColor=<rgba 0.875 1 0 1>}
  NamedRerouteDeclaration(None=FunctionInput() {inputName=RayUV, inputType=FunctionInput_Vector2, sortPriority=15}) {name=RayUV, nodeColor=<rgba 0.3828 1 0 1>}
  Reroute()
 ]}
