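# Material function that applies parallax occlusion mapping to an input UV (UV_TP) and returns the displaced UV.
# A custom HLSL loop ray-marches the height texture's red channel with explicit derivatives, interpolating the hit, from a start offset set by height strength and RefPlanar.
# Step count drops from MaxStep toward one through a FresnelVX blend of BlendFactor and noise-jittered BlendTolerance.
# Ray tracing skips the march, returning only the offset UV.
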MaterialFunction /InterchangeAxfAssets/MaterialFunctions/MF_POM_TP
graph {nodes [
  n1 = NamedRerouteUsage()
  n2 = Multiply(A=OneMinus(None=FunctionInput() {inputName=RefPlanar, inputType=FunctionInput_Scalar, sortPriority=14})) {constB=-1}
  n3 = FunctionInput() {bUsePreviewValueAsDefault=true, inputName=POM_HeightStrength, inputType=FunctionInput_Scalar, sortPriority=11, previewValue=<0.05 0 0 1>}
  n4 = Multiply(A=@n3) {constB=0.01}
  MaterialExpressionAdd_0 = Add(A=Multiply(A=Multiply(A=@n1, B=@n2), B=@n4), B=FunctionInput() {inputName=UV_TP, inputType=FunctionInput_Vector2, sortPriority=10})
  MaterialExpressionFunctionInput_3 = FunctionInput(Preview=TextureObject() {samplerType=SAMPLERTYPE_LinearColor, texture=/Engine/ArtTools/RenderToTexture/Textures/127grey}) {bUsePreviewValueAsDefault=true, inputName=POM_HeightTexture, inputType=FunctionInput_Texture2D}
  MaterialExpressionNamedRerouteUsage_6 = NamedRerouteUsage()
  MaterialExpressionNamedRerouteUsage_2 = NamedRerouteUsage()
  MaterialExpressionNamedRerouteUsage_3 = NamedRerouteUsage()
  MaterialExpressionFunctionInput_1 = FunctionInput() {inputName=DDX, inputType=FunctionInput_Vector2, sortPriority=15}
  MaterialExpressionFunctionInput_4 = FunctionInput() {inputName=DDY, inputType=FunctionInput_Vector2, sortPriority=16}
  MaterialExpressionConstant4Vector_0 = Constant4Vector() {constant=<rgba 1 0 0 0>}
  n5 = Custom!(
   Tex=@MaterialExpressionFunctionInput_3
   UV=@MaterialExpressionAdd_0
   UVDist=@MaterialExpressionNamedRerouteUsage_6
   MaxSteps=@MaterialExpressionNamedRerouteUsage_2
   stepsize=@MaterialExpressionNamedRerouteUsage_3
   InDDX=@MaterialExpressionFunctionInput_1
   InDDY=@MaterialExpressionFunctionInput_4
   HeightChannel=@MaterialExpressionConstant4Vector_0.RGBA
  ) {
   code="float rayheight = 1;\r\nfloat oldray = 1;\r\nfloat2 offset = 0;\r\nfloat oldtex = 1;\r\nfloat texAtRay;\r\nfloat yintersect;\r\nint i = 0;\r\n\r\nwhile (i < MaxSteps + 2)\r\n{\r\n\tfloat texAtRay = dot(Tex.SampleGrad( GetMaterialSharedSampler(TexSampler,View.MaterialTextureBilinearWrapedSampler), UV + offset, \r\n                     InDDX * (MaterialFloat2)View.MaterialTextureDerivativeMultiply, InDDY * (MaterialFloat2)View.MaterialTextureDerivativeMultiply ), HeightChannel);\r\n\r\n\tif (rayheight < texAtRay)\r\n\t{\r\n\t\tfloat xintersect = (oldray - oldtex) + (texAtRay - rayheight);\r\n\t\txintersect = (texAtRay - rayheight) / xintersect;\r\n\t\tyintersect = ( oldray * (xintersect) ) + ( rayheight * (1 - xintersect) );\r\n\t\toffset -= (xintersect * UVDist);\r\n\t\tbreak;\r\n\t}\r\n\r\n\toldray = rayheight;\r\n\trayheight -= stepsize;\r\n\toffset += UVDist;\r\n\toldtex = texAtRay;\r\n\r\n\ti++;\r\n}\r\n\r\nreturn UV + offset;"
   description=ParallaxMapping
   outputType=CMOT_Float2
   showCode=true
   inputs [
    {inputName=Tex, input {expression=&self:MaterialExpressionFunctionInput_3}}
    {inputName=UV, input {expression=&self:MaterialExpressionAdd_0}}
    {inputName=UVDist, input {expression=&self:MaterialExpressionNamedRerouteUsage_6}}
    {inputName=MaxSteps, input {expression=&self:MaterialExpressionNamedRerouteUsage_2}}
    {inputName=stepsize, input {expression=&self:MaterialExpressionNamedRerouteUsage_3}}
    {inputName=InDDX, input {expression=&self:MaterialExpressionFunctionInput_1}}
    {inputName=InDDY, input {expression=&self:MaterialExpressionFunctionInput_4}}
    {
     inputName=HeightChannel
     input {expression=&self:MaterialExpressionConstant4Vector_0, mask=1, maskA=1, maskB=1, maskG=1, maskR=1}
    }
   ]
  }
  FunctionOutput(None=RayTracingQualitySwitch(Normal=@n5, RayTraced=@MaterialExpressionAdd_0))
  NamedRerouteDeclaration(None=FunctionInput() {inputName=RayUV, inputType=FunctionInput_Vector2, sortPriority=18}) {name=RayUV, nodeColor=<rgba 0.1764 1 0.109 1>}
  n6 = Multiply(
   A=LinearInterpolate(
    A=Constant() {r=0.9}
    B=Constant() {r=1.2}
    Alpha=MaterialFunctionCall!(TemporalSampleIndex, PixelPosition) {materialFunction=&/MF_Noise_InterLeavedGradientGloden_1d}.Result
   ) {constA=0.9}
   B=FunctionInput() {inputName=BlendTolerance, inputType=FunctionInput_Scalar, sortPriority=23}
  )
  n7 = FunctionInput() {
   desc=" Blend Factor controls the position of the heighmap in the blend (in effect, the softness), and Blend "
   inputName=BlendFactor
   inputType=FunctionInput_Scalar
   sortPriority=22
  }
  n8 = FunctionInput() {inputName=FresnelVX, inputType=FunctionInput_Scalar, sortPriority=20}
  n9 = LinearInterpolate(
   A=FunctionInput(Preview=Constant() {r=32}) {inputName=MaxStep, inputType=FunctionInput_Scalar, sortPriority=12}
   B=Constant() {desc=MinStep, r=1}
   Alpha=Clamp(None=LinearInterpolate(A=@n6, B=@n7, Alpha=@n8))
  )
  NamedRerouteDeclaration(None=Divide(B=@n9)) {name=StepSize, nodeColor=<rgba 1 0.7266 0 1>}
  n10 = NamedRerouteUsage()
  NamedRerouteDeclaration(None=Multiply(A=Multiply(A=@n4, B=@n10), B=NamedRerouteUsage())) {name=UV_D, nodeColor=<rgba 0.4062 0 1 1>}
  NamedRerouteDeclaration(None=Floor(None=@n9)) {name=Maxstep, nodeColor=<rgba 0.875 1 0 1>}
 ]}
