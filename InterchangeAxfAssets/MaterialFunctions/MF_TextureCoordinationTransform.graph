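# Material function in the AxF Interchange assets that transforms texture coordinates with scale, per-axis stretch, rotation, offset and time-driven animation.
# UVs are scaled and shifted around a half-unit center, then rotated with sine and cosine.
# Offset and Time-driven movement are added, while rotation speed adds a Frac-wrapped spin.
# Named reroutes carry intermediates, including AxF X/Y parameters, and outputs expose coordinates, rotation and XY scale.
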MaterialFunction /InterchangeAxfAssets/MaterialFunctions/MF_TextureCoordinationTransform
graph {nodes [
  n1 = FunctionInput() {description=360 degree, inputName=Scale, inputType=FunctionInput_Scalar}
  n2 = AppendVector(
   A=FunctionInput() {description=360 degree, inputName=Stretch on X, inputType=FunctionInput_Scalar, sortPriority=1}
   B=FunctionInput() {description=360 degree, inputName=Stretch on Y, inputType=FunctionInput_Scalar, sortPriority=2}
  )
  FunctionOutput(None=NamedRerouteDeclaration(None=Multiply(A=@n1, B=@n2)) {name=Scale_XY, nodeColor=<rgba 0 0.25 1 1>}) {outputName=Scale_XY}
  n3 = FunctionInput() {description=360 degree, inputName=Rotation, inputType=FunctionInput_Scalar, sortPriority=3}
  n4 = Constant() {r=2}
  n5 = Divide(A=FunctionInput() {description=360 degree, inputName=Speed Rotation, inputType=FunctionInput_Scalar, sortPriority=8}) {constB=360}
  n6 = NamedRerouteUsage()
  n7 = Add(
   A=Divide(
    A=MaterialFunctionCall!(Degrees=@n3) {materialFunction=/Engine/Functions/Engine_MaterialFunctions01/Math/DegreesToRadians}.Radians
    B=MaterialFunctionCall(Multiplier=@n4) {materialFunction=&Pi}.Result
   )
   B=Frac(None=Multiply(A=@n5, B=@n6))
  )
  n8 = Cosine(None=@n7)
  n9 = Multiply(A=@n8, B=NamedRerouteUsage())
  n10 = Sine(None=@n7)
  n11 = Multiply(A=@n10, B=NamedRerouteUsage())
  n12 = Multiply(A=@n10, B=NamedRerouteUsage())
  n13 = Multiply(A=@n8, B=NamedRerouteUsage())
  n14 = Add(A=AppendVector(A=Subtract(A=@n9, B=@n11), B=Add(A=@n12, B=@n13)), B=Divide(A=Constant2Vector() {g=0.5, r=0.5}, B=NamedRerouteUsage()))
  n15 = NamedRerouteUsage()
  FunctionOutput(None=Add(A=Add(A=@n14, B=@n15), B=NamedRerouteUsage())) {outputName=TextureCoordinates}
  FunctionOutput(None=@n7) {outputName=Rotation}
  n16 = TextureCoordinate()
  n17 = NamedRerouteUsage()
  n18 = Add(A=Divide(A=Multiply(A=@n16, B=@n17), B=NamedRerouteUsage()), B=Divide(A=Constant2Vector() {g=-0.5, r=-0.5}, B=NamedRerouteUsage()))
  NamedRerouteDeclaration(None=ComponentMask(None=@n18) {r=true}) {name=UV_T_X, nodeColor=<rgba 0 1 0.6484 1>}
  n19 = FunctionInput() {description=360 degree, inputName=AXF_X, inputType=FunctionInput_Scalar, sortPriority=9}
  n20 = FunctionInput() {description=360 degree, inputName=AXF_Y, inputType=FunctionInput_Scalar, sortPriority=10}
  NamedRerouteDeclaration(None=Multiply(A=AppendVector(A=@n19, B=@n20)) {constB=100}) {name=AXF_Parameters, nodeColor=<rgba 1 0.6797 0 1>}
  NamedRerouteDeclaration(None=Time() {bOverride_Period=true, period=9999999}) {name=Timer, nodeColor=<rgba 1 0 0.9375 1>}
  NamedRerouteDeclaration(None=ComponentMask(None=@n18) {g=true}) {name=UV_T_Y, nodeColor=<rgba 0 1 0.7422 1>}
  n21 = FunctionInput() {description=360 degree, inputName=Offset on X, inputType=FunctionInput_Scalar, sortPriority=4}
  n22 = FunctionInput() {description=360 degree, inputName=Offset on Y, inputType=FunctionInput_Scalar, sortPriority=5}
  n23 = ComponentMask(None=NamedRerouteUsage()) {g=true}
  NamedRerouteDeclaration(None=AppendVector(A=Multiply(A=@n21) {constB=-1}, B=Subtract(A=@n22, B=@n23))) {name=Offset, nodeColor=<rgba 0.3828 0 1 1>}
  n24 = Multiply(A=FunctionInput() {description=360 degree, inputName=Speed on X, inputType=FunctionInput_Scalar, sortPriority=6}) {constB=-1}
  n25 = FunctionInput() {description=360 degree, inputName=Speed on Y, inputType=FunctionInput_Scalar, sortPriority=7}
  NamedRerouteDeclaration(None=Multiply(A=AppendVector(A=@n24, B=@n25), B=NamedRerouteUsage())) {name=Movment, nodeColor=<rgba 0 1 0.0625 1>}
 ]}
