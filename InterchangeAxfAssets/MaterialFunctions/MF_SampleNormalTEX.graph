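# Material function in the Interchange AxF assets that samples a normal map with explicit derivatives and an optional mask to cheapen masked-out samples.
# A Custom HLSL node samples with Texture2DSampleGrad and a shared bilinear wrapped sampler, derivatives defaulting to those of the coordinates.
# Near-zero Mask branches to DefaultUV with zero derivatives.
# Outputs give the unpacked normal, its channels, and raw PackedRGBA.
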MaterialFunction /InterchangeAxfAssets/MaterialFunctions/MF_SampleNormalTEX
graph {nodes [
  MaterialExpressionFunctionInput_1 = FunctionInput(Preview=TextureObject() {samplerType=SAMPLERTYPE_Normal, texture=/Engine/EngineMaterials/BaseFlattenNormalMap}) {bUsePreviewValueAsDefault=true, inputName=TexObject, inputType=FunctionInput_Texture2D, sortPriority=1}
  MaterialExpressionFunctionInput_0 = FunctionInput(Preview=TextureCoordinate()) {bUsePreviewValueAsDefault=true, inputName=Coordinates, inputType=FunctionInput_Vector2}
  MaterialExpressionFunctionInput_2 = FunctionInput(Preview=DDX(Value=@MaterialExpressionFunctionInput_0)) {bUsePreviewValueAsDefault=true, inputName=DDX, inputType=FunctionInput_Vector2, sortPriority=2}
  MaterialExpressionFunctionInput_3 = FunctionInput(Preview=DDY(Value=@MaterialExpressionFunctionInput_0)) {bUsePreviewValueAsDefault=true, inputName=DDY, inputType=FunctionInput_Vector2, sortPriority=3}
  MaterialExpressionFunctionInput_4 = FunctionInput(Preview=Constant() {r=1}) {bUsePreviewValueAsDefault=true, inputName=Mask, inputType=FunctionInput_Scalar, sortPriority=4}
  MaterialExpressionFunctionInput_5 = FunctionInput(Preview=Constant2Vector() {g=0.5, r=0.5}) {bUsePreviewValueAsDefault=true, inputName=DefaultUV, inputType=FunctionInput_Vector2, sortPriority=5}
  n1 = Custom!(
   Texture=@MaterialExpressionFunctionInput_1
   UV=@MaterialExpressionFunctionInput_0
   DDX=@MaterialExpressionFunctionInput_2
   DDY=@MaterialExpressionFunctionInput_3
   Mask=@MaterialExpressionFunctionInput_4
   Default=@MaterialExpressionFunctionInput_5
  ) {
   code="[branch] if( Mask < 0.0001f) \r\n{\r\n    UV = Default;\r\n    DDX *= 0;\r\n    DDY *= 0;\r\n}\r\nMaterialFloat4 OutTextureValue = Texture2DSampleGrad(\r\n    Texture,\r\n    GetMaterialSharedSampler(TextureSampler,View.MaterialTextureBilinearWrapedSampler),\r\n    UV,\r\n    DDX * (MaterialFloat2)View.MaterialTextureDerivativeMultiply,\r\n    DDY * (MaterialFloat2)View.MaterialTextureDerivativeMultiply);\r\nPackedRGBA = OutTextureValue;\r\nOutTextureValue.xyz = UnpackNormalMap(OutTextureValue).xyz;\r\nreturn OutTextureValue;"
   desc=https://ryandowlingsoka.com/unreal/triplanar-dither-biplanar/
   description=PointTSNSample
   outputType=CMOT_Float4
   showCode=true
   additionalOutputs [{outputName=PackedRGBA, outputType=CMOT_Float4}]
   inputs [
    {inputName=Texture, input {expression=&self:MaterialExpressionFunctionInput_1}}
    {inputName=UV, input {expression=&self:MaterialExpressionFunctionInput_0}}
    {inputName=DDX, input {expression=&self:MaterialExpressionFunctionInput_2}}
    {inputName=DDY, input {expression=&self:MaterialExpressionFunctionInput_3}}
    {inputName=Mask, input {expression=&self:MaterialExpressionFunctionInput_4}}
    {inputName=Default, input {expression=&self:MaterialExpressionFunctionInput_5}}
   ]
  }
  FunctionOutput(None=ComponentMask(None=@n1.return) {b=true, g=true, r=true}) {outputName=RGB}
  FunctionOutput(None=ComponentMask(None=@n1.return) {r=true}) {outputName=R, sortPriority=1}
  FunctionOutput(None=ComponentMask(None=@n1.return) {g=true}) {outputName=G, sortPriority=2}
  FunctionOutput(None=ComponentMask(None=@n1.return) {b=true}) {outputName=B, sortPriority=3}
  FunctionOutput(None=ComponentMask(None=@n1.return) {a=true, b=true, g=true, r=true}) {outputName=RGBA, sortPriority=4}
  FunctionOutput(None=ComponentMask(None=@n1.PackedRGBA) {a=true, b=true, g=true, r=true}) {outputName=PackedRGBA, sortPriority=5}
 ]}
