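# MF_SafeNormalize is a material function in the AxF Interchange assets that normalizes a vector, with a fallback for zero length.
# A single If compares the squared length of Normal to zero, selecting either the normalized vector or the Default input.
# Each branch packs a flag into a fourth channel, which masks split into the RGB vector output and a "Length==0" output that is 1 on fallback.
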MaterialFunction /InterchangeAxfAssets/MaterialFunctions/MF_SafeNormalize
graph {nodes [
  n1 = FunctionInput() {inputName=Normal}
  n2 = AppendVector(A=FunctionInput() {inputName=Default, sortPriority=1}, B=Constant() {r=1})
  n3 = If(A=DotProduct(A=@n1, B=@n1), B=Constant() {r=0}, "A > B"=AppendVector(A=Normalize(VectorInput=@n1), B=Constant()), "A == B"=@n2, "A < B"=@n2)
  FunctionOutput(None=ComponentMask(None=@n3) {b=true, g=true, r=true})
  FunctionOutput(None=ComponentMask(None=@n3) {a=true}) {outputName="Length==0", sortPriority=1}
 ]}
