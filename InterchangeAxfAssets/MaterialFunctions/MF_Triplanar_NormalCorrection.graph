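# Material function for AxF triplanar normal correction, converting a normal defined in a supplied tangent frame into world-space and tangent-space normals.
# Tangent and Bitangent inputs pass through safe normalization with fallback axes into named reroutes.
# The NormalT components weight that basis and the world vertex normal; the normalized sum is output as WSN.
# The same vector, transformed from world to tangent space, is output as TSN.
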MaterialFunction /InterchangeAxfAssets/MaterialFunctions/MF_Triplanar_NormalCorrection
graph {nodes [
  n1 = MaterialFunctionCall(Float3=FunctionInput(Preview=Constant3Vector() {constant=<rgba 0 0 1 0>}) {bUsePreviewValueAsDefault=true, inputName=NormalT}) {materialFunction=&BreakOutFloat3Components}
  n2 = Multiply(A=NamedRerouteUsage(), B=@n1.R)
  n3 = Multiply(A=NamedRerouteUsage(), B=@n1.G)
  n4 = VertexNormalWS()
  n5 = Normalize(VectorInput=Add(A=Add(A=@n2, B=@n3), B=Multiply(A=@n4, B=@n1.B)) {desc=WS})
  FunctionOutput(None=Transform(None=@n5) {transformSourceType=TRANSFORMSOURCE_World, transformType=TRANSFORM_Tangent}) {outputName=TSN}
  FunctionOutput(None=@n5) {outputName=WSN}
  n6 = Constant3Vector() {constant=<rgba 1 0 0 0>}
  NamedRerouteDeclaration(None=MaterialFunctionCall!(Normal=FunctionInput(Preview=@n6) {bUsePreviewValueAsDefault=true, inputName=Tangent, sortPriority=1}, Default=@n6) {materialFunction=&/MF_SafeNormalize}.Result) {name=TangentWS, nodeColor=<rgba 0.875 0 1 1>}
  n7 = Constant3Vector() {constant=<rgba 0 -1 0 0>}
  NamedRerouteDeclaration(None=MaterialFunctionCall(Normal=FunctionInput(Preview=@n7) {bUsePreviewValueAsDefault=true, inputName=Bitangent, sortPriority=2}, Default=@n7) {materialFunction=&MF_SafeNormalize}.Result) {name=BitangentWs, nodeColor=<rgba 0 1 0.5547 1>}
 ]}
