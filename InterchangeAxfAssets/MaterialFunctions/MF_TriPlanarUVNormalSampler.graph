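# Material function from InterchangeAxfAssets that samples and corrects a normal map for triplanar UV projection.
# MF_SampleNormalTEX reads the texture at the supplied UV with DDX and DDY inputs.
# MF_Triplanar_NormalCorrection then combines that RGB normal with the supplied tangent and bitangent.
# A custom HLSL node rotates the resulting TSN vector around Z, mapping the saturated rotation input to a full turn, to produce NormalOutput.
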MaterialFunction /InterchangeAxfAssets/MaterialFunctions/MF_TriPlanarUVNormalSampler
graph {nodes [
  MaterialExpressionFunctionInput_9 = FunctionInput() {inputName=Rotation_TP, inputType=FunctionInput_Scalar, sortPriority=6}
  n1 = TextureCoordinate()
  n2 = TextureObject() {samplerType=SAMPLERTYPE_Normal, texture=/Bridge/MSPresets/MSTextures/FlatNormal}
  MaterialExpressionMaterialFunctionCall_2 = MaterialFunctionCall!(
   NormalT=MaterialFunctionCall!(
    Coordinates=FunctionInput(Preview=@n1) {inputName=UV_TP, inputType=FunctionInput_Vector2, sortPriority=3}
    TexObject=FunctionInput(Preview=@n2) {inputName=NormalTexOBJ, inputType=FunctionInput_Texture2D, sortPriority=7}
    DDX=FunctionInput() {inputName=DDX, inputType=FunctionInput_Scalar}
    DDY=FunctionInput() {inputName=DDY, inputType=FunctionInput_Scalar, sortPriority=1}
    Mask
    DefaultUV
   ) {materialFunction=&/MF_SampleNormalTEX}.RGB
   Tangent=FunctionInput() {inputName=Tangent, sortPriority=4}
   Bitangent=FunctionInput() {inputName=Bitangent, sortPriority=5}
  ) {materialFunction=&/MF_Triplanar_NormalCorrection}
  n3 = Custom!(RotationAmount=@MaterialExpressionFunctionInput_9, Vector=@MaterialExpressionMaterialFunctionCall_2.TSN) {
   code="const float theta = saturate(RotationAmount) * PI * 2.;\r\nconst float3x3 rot3M = { cos(theta) , sin(theta), 0.f,\r\n                         -sin(theta), cos(theta), 0.f,\r\n                                 0.f,        0.f, 1.f\r\n                       };\r\nreturn mul(rot3M, Vector);"
   description=RotCW3D_Plane
   showCode=true
   inputs [
    {
     inputName=RotationAmount
     input {expression=&self:MaterialExpressionFunctionInput_9}
    }
    {
     inputName=Vector
     input {expression=&self:MaterialExpressionMaterialFunctionCall_2, outputIndex=1}
    }
   ]
  }
  FunctionOutput(None=@n3) {outputName=NormalOutput}
 ]}
