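# Material function for AxF import materials that computes a view-ray UV offset per unit depth (RayUV) for a triplanar projection.
# The camera vector (view-space Z axis in the vertex stage) is transformed into a basis of the input triplanar tangent, bitangent and vertex world normal.
# Its negated XY is divided by Z, parallax-style, then rotated about the origin by the Rotation input.
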MaterialFunction /InterchangeAxfAssets/MaterialFunctions/MF_ViewTaceLength_TP
graph {nodes [
  n1 = Constant3Vector() {constant=<rgba 0 0 1 0>}
  n2 = MaterialFunctionCall!(
   VectorToTransform=ShaderStageSwitch(PixelShader=CameraVectorWS(), VertexShader=Transform(None=@n1) {transformSourceType=TRANSFORMSOURCE_View})
   BasisX=FunctionInput() {inputName=triplanar_tangent}
   BasisY=FunctionInput() {inputName=triplanar_bitangent}
   BasisZ=VertexNormalWS()
  ) {materialFunction=/Engine/Functions/Engine_MaterialFunctions02/Math/InverseTransformMatrix}
  n3 = Multiply(A=ComponentMask(None=@n2.Transformed Vector 3) {g=true, r=true}) {constB=-1}
  n4 = ComponentMask(None=@n2.Transformed Vector 3) {b=true}
  FunctionOutput(None=MaterialFunctionCall!(
    RotationAmount=FunctionInput() {inputName=Rotation, inputType=FunctionInput_Scalar, sortPriority=6}
    Coords=Divide(A=@n3, B=@n4) {constB=1}
    Pivot=Constant2Vector()
   ) {materialFunction=&/MF_Rotate2D}.Result) {outputName=RayUV}
 ]}
