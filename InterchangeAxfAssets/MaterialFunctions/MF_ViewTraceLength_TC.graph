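# Material function from the AxF Interchange assets that calculates the view ray in UV space, output as Ray_UV.
# The camera vector is converted to tangent space, or stays in world space when Use World Coordinates is on, optionally re-based on the vertex normal.
# Its negated XY is divided by Z to give the UV offset per unit depth, then rotated about the origin by the Rotation input via MF_Rotate2D.
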MaterialFunction /InterchangeAxfAssets/MaterialFunctions/MF_ViewTraceLength_TC
graph {nodes [
  n1 = CameraVectorWS()
  n2 = VertexNormalWS()
  n3 = FunctionInput(Preview=StaticBool() {value=true}) {
   bUsePreviewValueAsDefault=true
   inputName=Transform To VertexNormal
   inputType=FunctionInput_StaticBool
   sortPriority=50
   previewValue=<1 1 1 1>
  }
  n4 = FunctionInput(Preview=StaticBool()) {
   bUsePreviewValueAsDefault=true
   description="If using world coordinates as the UVs, the shader needs to skip the tangent->world transform. This saves a few instructions. Default = FALSE"
   inputName=Use World Coordinates
   inputType=FunctionInput_StaticBool
   sortPriority=11
  }
  n5 = StaticSwitch(
   True=StaticSwitch(
    True=MaterialFunctionCall!(Vector to Transform=@n1, Z Vector=@n2, Center Location) {materialFunction=/Engine/Functions/Engine_MaterialFunctions02/Math/TransformToZVector}.Result
    False=@n1
    Value=@n3
   )
   False=Transform(None=@n1) {transformSourceType=TRANSFORMSOURCE_World, transformType=TRANSFORM_Tangent}
   Value=@n4
  )
  n6 = Multiply(A=ComponentMask(None=@n5) {g=true, r=true}) {constB=-1}
  n7 = ComponentMask(None=@n5) {b=true}
  FunctionOutput(None=MaterialFunctionCall!(
    RotationAmount=FunctionInput() {inputName=Rotation, inputType=FunctionInput_Scalar, sortPriority=6}
    Coords=Divide(A=@n6, B=@n7) {constB=1}
    Pivot=Constant2Vector()
   ) {materialFunction=&/MF_Rotate2D}.Result) {description=Caculate Ray on UV space, outputName=Ray_UV}
 ]}
